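# Physics asset for the hair of the F_MED_Tactical_Red_Punk female character, built from capsule bodies on dyn_hair bones.
# Front strands branch from the head, while middle and back chains hang off a boxed dyn_hair_mid_MSTR tied to neck_01; three single-bone pieces attach to the head.
# Hair joints use tight swing and twist limits, with one locked swing axis.
# Head, neck, right clavicle and right upper arm add body shapes with free angular joints.
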PhysicsAsset /BRCosmetics/Characters/Player/Female/Medium/Bodies/F_MED_Tactical_Red_Punk/Meshes/Parts/F_MED_Tactical_Red_Punk_Physics
bodies {
 clavicle_r [{shapeType=Capsule, length=5.7263, radius=5.2737, center=<-6.3269 1.3288 -0.5936>, rotation=<rot 87.1873 0 0>}]
 dyn_hair_back [{shapeType=Capsule, length=0, radius=0.505, center=<-0.7868 0 0>}]
 dyn_hair_back_01 [{shapeType=Capsule, length=2.9899, radius=1.0114, center=<2.5263 0 0.0003>, rotation=<rot 90 0 0>}]
 dyn_hair_back_02 [{shapeType=Capsule, length=2.816, radius=0.956, center=<1.8156 -0.1007 0.5308>, rotation=<rot 82.3751 50.3167 73.2254>}]
 dyn_hair_back_low [{shapeType=Capsule, length=0, radius=0.603, center=<-0.5898 0 0>}]
 dyn_hair_front_01 [{shapeType=Capsule, length=1.9887, radius=2.9327, center=<3.4553 -0.0001 0>, rotation=<rot 90 1.2655 1.2652>}]
 dyn_hair_front_02 [{shapeType=Capsule, length=1.9305, radius=1.8398, center=<2.8398 -0.0001 0.0001>, rotation=<rot 90 0.0299 0.0302>}]
 dyn_hair_front_02a [{shapeType=Capsule, length=0, radius=0.8909, center=<1.0361 -0.0001 0>}]
 dyn_hair_front_02b [{shapeType=Capsule, length=0, radius=0.75, center=<1.2569 -0.0001 0>}]
 dyn_hair_front_03 [{shapeType=Capsule, length=2.4674, radius=1.2263, center=<2.5613 0.255 0.918>, rotation=<rot -55.3511 0 0>}]
 dyn_hair_mid_01 [{shapeType=Capsule, length=0.318, radius=3.0963, center=<3.4497 0.0001 0>, rotation=<rot 90 -0.0098 -0.0098>}]
 dyn_hair_mid_02 [
  {shapeType=Capsule, length=4.8011, radius=1.6989, center=<2.7745 0 0.0003>, rotation=<rot 90 180 -89.656>}
  {
   shapeType=Capsule
   length=3.812
   radius=2.2219
   shapeName=dyn_hair_mid_02_capsule
   center=<6.8729 -0.0417 0.5816>
   rotation=<rot 0 0.0001 85.9018>
  }
 ]
 dyn_hair_mid_03 [{shapeType=Capsule, length=4.9685, radius=1.1926, center=<3.6063 -0.4755 -1.1194>, rotation=<rot 54.9998 0 0>}]
 dyn_hair_mid_MSTR [{shapeType=Box, extentX=1.5, extentY=1.5, extentZ=2.25}]
 dyn_hair_mid_high [{shapeType=Capsule, length=0, radius=1, center=<-1.1811 0 0.4596>}]
 head [{shapeType=Capsule, length=10.7384, radius=7.3938, rotation=<rot -90 -0.02 0.0202>}]
 neck_01 [{shapeType=Capsule, length=1.5631, radius=5.5747, shapeName=neck_01_capsule, rotation=<rot 90 0 0>}]
 upperarm_r [{
   shapeType=Capsule
   length=12.067
   radius=4.1282
   shapeName=upperarm_r_capsule
   center=<-10.163 0.4989 0.637>
   rotation=<rot -0.0449 -91.5839 -87.6299>
  }]
}
constraints [
 {bone1Name=dyn_hair_mid_high, bone2Name=head, swing1LimitDegrees=15, swing2LimitDegrees=15, twistLimitDegrees=2}
 {bone1Name=dyn_hair_back_low, bone2Name=head, swing1LimitDegrees=20, swing2LimitDegrees=15, twistLimitDegrees=2}
 {bone1Name=dyn_hair_back, bone2Name=head, swing1LimitDegrees=10, swing2LimitDegrees=10, twistLimitDegrees=2}
 {bone1Name=dyn_hair_mid_03, bone2Name=dyn_hair_mid_02, swing1LimitDegrees=15, swing2LimitDegrees=15, swing2Motion=Locked, twistLimitDegrees=2}
 {bone1Name=dyn_hair_front_01, bone2Name=head, swing1LimitDegrees=15, swing2LimitDegrees=15, twistLimitDegrees=2}
 {bone1Name=dyn_hair_front_02, bone2Name=dyn_hair_front_01, swing1LimitDegrees=10, swing2LimitDegrees=20, twistLimitDegrees=2}
 {bone1Name=dyn_hair_front_03, bone2Name=dyn_hair_front_02, swing1LimitDegrees=25, swing2LimitDegrees=25, twistLimitDegrees=2}
 {bone1Name=dyn_hair_front_02a, bone2Name=dyn_hair_front_01, swing1LimitDegrees=15, swing2LimitDegrees=15, twistLimitDegrees=2}
 {bone1Name=dyn_hair_front_02b, bone2Name=dyn_hair_front_01, swing1LimitDegrees=15, swing2LimitDegrees=15, twistLimitDegrees=2}
 {bone1Name=dyn_hair_mid_01, bone2Name=dyn_hair_mid_MSTR, swing1LimitDegrees=35, swing2LimitDegrees=35, twistLimitDegrees=5}
 {bone1Name=dyn_hair_mid_02, bone2Name=dyn_hair_mid_01, swing1LimitDegrees=15, swing2LimitDegrees=15, twistLimitDegrees=10}
 {bone1Name=dyn_hair_back_02, bone2Name=dyn_hair_back_01, swing1LimitDegrees=10, swing2LimitDegrees=10, twistLimitDegrees=2}
 {bone1Name=dyn_hair_back_01, bone2Name=dyn_hair_mid_01, swing1LimitDegrees=20, swing2LimitDegrees=20, twistLimitDegrees=2}
 {
  bone1Name=upperarm_r
  bone2Name=clavicle_r
  swing1LimitDegrees=45
  swing1Motion=Free
  swing2LimitDegrees=45
  swing2Motion=Free
  twistLimitDegrees=45
  twistMotion=Free
 }
 {
  bone1Name=head
  bone2Name=neck_01
  swing1LimitDegrees=45
  swing1Motion=Free
  swing2LimitDegrees=45
  swing2Motion=Free
  twistLimitDegrees=45
  twistMotion=Free
 }
 {
  bone1Name=dyn_hair_mid_MSTR
  bone2Name=neck_01
  swing1LimitDegrees=45
  swing1Motion=Free
  swing2LimitDegrees=45
  swing2Motion=Free
  twistLimitDegrees=45
  twistMotion=Free
 }
]
