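# Animation Blueprint for a female Tactical Red Punk character part that follows the body pose, reduces neck and hair clipping, and simulates physics.
# Copied mesh poses, blended at upperarm_r, pass through cached shoulder and neck corrective bone stages, then hair offsets and RigidBody simulation.
# The EventGraph derives corrective alphas from bone deltas against reference pose, with hair offsets for ADS and open parachute.
# A deprecated state function supplies the RigidBody gravity override.
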AnimBlueprint /BRCosmetics/Characters/Player/Female/Medium/Bodies/F_MED_Tactical_Red_Punk/Meshes/Parts/F_MED_Tactical_Red_Punk_AnimBP
target_skeleton=&/F_MED_Tactical_Red_Punk_Skeleton
graphs [
 AnimGraph {nodes [
   n1 = "Transform (Modify) Bone - Bone: dyn_hair_mid_MSTR"(
    ComponentPose=Local To Component(LocalPose=Use cached pose 'neck_body_tweaks_cache'())
    Alpha=Get ADS_Fix() -> (ADS_Fix:double)
    Translation="1.1,0,0"
    Rotation="-30,15,25"
    Scale="1,1,1"
   ) {
    rotationMode=BMM_Additive
    rotationSpace=BCS_BoneSpace
    translationMode=BMM_Additive
    translationSpace=BCS_BoneSpace
    alphaBoolBlend {blendOption=Linear}
    alphaScaleBiasClamp {bInterpResult=true, interpSpeedDecreasing=7, interpSpeedIncreasing=7}
    boneToModify {boneName=dyn_hair_mid_MSTR}
    rotation=<rot -30 15 25>
    translation {x=1.1}
   }
   n2 = "Transform (Modify) Bone - Bone: dyn_hair_mid_MSTR"(
    ComponentPose=@n1
    Alpha=Get avoid_clipping_hair_parachuting() -> (avoid_clipping_hair_parachuting:double)
    Translation="0,0,0"
    Rotation="-25,-10,0"
    Scale="1,1,1"
   ) {
    rotationMode=BMM_Additive
    rotationSpace=BCS_BoneSpace
    translationMode=BMM_Additive
    translationSpace=BCS_BoneSpace
    alphaBoolBlend {blendOption=Linear}
    alphaScaleBiasClamp {bInterpResult=true}
    boneToModify {boneName=dyn_hair_mid_MSTR}
    rotation {pitch=-25, yaw=-10}
   }
   n3 = RigidBody!(ComponentPose=@n2, Alpha, OverrideWorldGravity:Vector=Get GravityOverride(), ExternalForce) -> (Pose) {
    bOverrideWorldGravity=true
    overridePhysicsAsset=&/F_MED_Tactical_Red_Punk_Physics
    simulationSpace=BaseBoneSpace
    alphaBoolBlend {blendOption=Linear}
    baseBoneRef {boneName=spine_05}
    componentAppliedLinearAccClamp=<2000 2000 2000>
    componentLinearAccScale=<0.25 0.25 0.3>
    componentLinearVelScale=<0.35 0.35 0.2>
    overrideWorldGravity {z=-980}
   }
   Output Pose(Result=Component To Local(ComponentPose=@n3)) {layerGroup=None}
   n4 = "Transform (Modify) Bone - Bone: base_neck_tweak_shoulder_l"(
    ComponentPose=Local To Component(LocalPose=Use cached pose 'blend_per_bone_cache'())
    Alpha=Get avoid_clipping_side_neck_l() -> (avoid_clipping_side_neck_l:double)
    Translation="0,0,0"
    Rotation="18,0,-18"
    Scale="1,1,1"
   ) {
    rotationMode=BMM_Additive
    rotationSpace=BCS_BoneSpace
    alphaBoolBlend {blendOption=Linear}
    boneToModify {boneName=base_neck_tweak_shoulder_l}
    rotation {pitch=18, roll=-18}
   }
   n5 = "Transform (Modify) Bone - Bone: base_neck_tweak_shoulder_r"(
    ComponentPose=@n4
    Alpha=Get avoid_clipping_side_neck_r() -> (avoid_clipping_side_neck_r:double)
    Translation="0,0,0"
    Rotation="-25,0,-18"
    Scale="1,1,1"
   ) {
    rotationMode=BMM_Additive
    rotationSpace=BCS_BoneSpace
    alphaBoolBlend {blendOption=Linear}
    boneToModify {boneName=base_neck_tweak_shoulder_r}
    rotation {pitch=-25, roll=-18}
   }
   Save cached pose 'neck_shoulders_tweak_cache'(Pose=Component To Local(ComponentPose=@n5)) {cacheName=neck_shoulders_tweak_cache}
   n6 = Get Body_Mesh()
   Save cached pose 'blend_per_bone_cache'(Pose=Layered blend per bone(
     BasePose=Copy Pose From Mesh() {bCopyCurves=true, bUseAttachedParent=true}
     BlendPoses_0=Copy Pose From Mesh(SourceMeshComponent=@n6) {bCopyCurves=true, bUseAttachedParent=true}
    ) {layerSetup [{branchFilters [{boneName=upperarm_r}]}]}) {cacheName=blend_per_bone_cache}
   n7 = "Transform (Modify) Bone - Bone: neck_tweak_back"(
    ComponentPose=Local To Component(LocalPose=Use cached pose 'neck_shoulders_tweak_cache'())
    Alpha=Get Alpha Output Fwd() -> (Alpha Output Fwd:double)
    Translation="0,0,0"
    Rotation="0,0,0"
    Scale="1,1.15,1.1"
   ) {
    scaleMode=BMM_Additive
    scaleSpace=BCS_BoneSpace
    alphaBoolBlend {blendOption=Linear}
    boneToModify {boneName=neck_tweak_back}
    scale {y=1.15, z=1.1}
   }
   n8 = "Transform (Modify) Bone - Bone: neck_tweak_front"(ComponentPose=@n7, Alpha=Get Alpha Output Bkwd() -> (Alpha Output Bkwd:double), Translation="0,0,0", Rotation="0,0,0", Scale="1,1.3,1.1") {
    scaleMode=BMM_Additive
    scaleSpace=BCS_BoneSpace
    alphaBoolBlend {blendOption=Linear}
    boneToModify {boneName=neck_tweak_front}
    scale {y=1.3, z=1.1}
   }
   n9 = Get Alpha Sides() -> (Alpha Sides:double)
   n10 = "Transform (Modify) Bone - Bone: neck_tweak_front"(ComponentPose=@n8, Alpha=@n9, Translation="0,0,0", Rotation="0,0,0", Scale="1,1.2,1.2") {
    scaleMode=BMM_Additive
    scaleSpace=BCS_BoneSpace
    alphaBoolBlend {blendOption=Linear}
    boneToModify {boneName=neck_tweak_front}
    scale {y=1.2, z=1.2}
   }
   n11 = "Transform (Modify) Bone - Bone: neck_tweak_back"(ComponentPose=@n10, Alpha=@n9, Translation="0,0,0", Rotation="0,0,0", Scale="1,1.2,1.2") {
    scaleMode=BMM_Additive
    scaleSpace=BCS_BoneSpace
    alphaBoolBlend {blendOption=Linear}
    boneToModify {boneName=neck_tweak_back}
    scale {y=1.2, z=1.2}
   }
   Save cached pose 'neck_body_tweaks_cache'(Pose=Component To Local(ComponentPose=@n11)) {cacheName=neck_body_tweaks_cache}
  ]}
 EventGraph {nodes [
   Event BlueprintUpdateAnimation()
   > n1 = Sequence!(execute) -> (then_0, then_1, then_2:Exec, then_3:Exec, then_4:Exec)
   n2 = GetOwnerSkeletalMeshComponent()
   n3 = BreakRotator(InRot=BreakTransform(InTransform=GetDeltaTransformFromRefPose(self=@n2, BoneName=neck_01, BaseName=spine_05)).Rotation)
   n4 = MapRangeClamped(Value=@n3.Yaw, InRangeA=40, InRangeB=-20, OutRangeA=-1)
   n5 = Branch(execute=@n1.then_1, Condition="float <= float"!(A:double=@n4, B:double=0.2) -> (ReturnValue:bool) {is=title K2Node_CallFunction})
   n6 = "Absolute (Float)"(A=@n4)
   > n7 = Set Alpha Output Fwd(Alpha Output Fwd:double=@n6) -> (Output_Get:double)
   Set Alpha Sides(
    execute=[@n7.then, @n8.then]
    Alpha Sides:double="Absolute (Float)"(A=MapRangeClamped(Value=@n3.Pitch, InRangeA=25, InRangeB=7, OutRangeA=-1))
   ) -> (Output_Get:double)
   n8 = Set Alpha Output Bkwd(execute=@n5.else, Alpha Output Bkwd:double=@n6) -> (Output_Get:double)
   n9 = Gravity Override State Function DEPRECATED(
    execute=@n1.then_3
    CharacterPartAnimInstance=Self-Reference()
    GameTimeInSeconds=GetGameTimeInSeconds()
    GravityOverrideParams=Get Gravity Override Params()
   )
   > Set GravityOverride(GravityOverride=@n9)
   Set avoid_clipping_hair_parachuting(execute=@n1.then_4, avoid_clipping_hair_parachuting:double="To Float (Boolean)"(InBool=Get bIsParachuteOpened())) -> (Output_Get:double)
   n10 = BreakTransform(InTransform=GetDeltaTransformFromRefPose(self=GetOwnerSkeletalMeshComponent(), BoneName=upperarm_r, BaseName=clavicle_r))
   Set ADS_Fix(execute=@n1.then_2, ADS_Fix:double=MapRangeClamped(Value=BreakRotator(InRot=@n10.Rotation).Yaw, InRangeA=70, InRangeB=80)) -> (Output_Get:double)
   n11 = Set Body_Mesh(execute=@n1.then_0, Body_Mesh=GetOwnerSkeletalMeshComponent())
   > Set avoid_clipping_side_neck_l(avoid_clipping_side_neck_l:double=GetAlphaFromReferenceTransformDelta(SkeletalMeshComponent=@n11, BoneName=clavicle_l, BaseName=upperarm_l, DeltaComponentType=EulerY, InputRange=(X=0,Y=-36), TwistAxis="0,0,1")) -> (Output_Get:double)
   > Set avoid_clipping_side_neck_r(avoid_clipping_side_neck_r:double=GetAlphaFromReferenceTransformDelta(SkeletalMeshComponent=@n11, BoneName=clavicle_r, BaseName=upperarm_r, DeltaComponentType=EulerY, InputRange=(X=0,Y=-36), TwistAxis="0,0,1")) -> (Output_Get:double)
  ]}
]
variables [
 ADS_Fix {type=number}
 Alpha Output Bkwd {type=number}
 Alpha Output Fwd {type=number}
 Alpha Sides {type=number}
 avoid_clipping_hair_parachuting {type=number}
 avoid_clipping_side_neck_l {type=number}
 avoid_clipping_side_neck_r {type=number}
 Body_Mesh {type=SkeletalMeshComponent}
 Gravity Override Params {type=GravityOverrideParamsStruct, default {skydive_Parachuting {gravityOverride=<0 0 -980>}}}
 GravityOverride {type=Vector}
]
